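; Spawn action that grants configured gameplay tags to spawned AI or actors, either before or after spawning, based on BeforeSpawn.
; Before spawn, it gets or creates the AI bot gameplay ability spawner component, appends the tags as replicated or loose per ShouldReplicate, then overrides that component in the list.
; After spawn, it adds loose tags to each valid spawned actor, passing ShouldReplicate as the replication flag.
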(fn ModifySpawnComponentList
  (params (ComponentList :type Fort%20Athena%20AISpawner%20Data%20Component%20List%20Object%20Reference) (ComponentOuter :type Object%20Reference))
  (bind _gameplay_tags (Variables|Default|GetGameplay%20Tags))
  (if (Variables|Default|GetBeforeSpawn)
    (bind _asfort_athena_aispawner_data_component_gameplay_ability_base (Utilities|Casting|CastToFortAthenaAISpawnerDataComponent_GameplayAbilityBase (Class|FortAthenaAISpawnerDataComponentList|GetorCreateComponentClassforModification ComponentList ComponentOuter "/Script/FortniteAI.FortAthenaAISpawnerDataComponent_AIBotGameplayAbilityBase")))
    (if (Variables|Default|GetShouldReplicate)
      (Class|FortAthenaAISpawnerDataComponentGameplayAbilityBase|AppendReplicatedTagstoApplytoPawn _asfort_athena_aispawner_data_component_gameplay_ability_base _gameplay_tags)
      (else
        (Class|FortAthenaAISpawnerDataComponentGameplayAbilityBase|AppendLooseTagstoApplytoPawn _asfort_athena_aispawner_data_component_gameplay_ability_base _gameplay_tags)))
    (AI|OverrideAISpawnerDataComponent 0 _asfort_athena_aispawner_data_component_gameplay_ability_base ComponentList)))

(fn OnAllActorSpawned
  (params (SpawnedActor :type Array%20of%20Actor%20Object%20References) (SourcePointProvider :type Fort%20Athena%20Living%20World%20Point%20Provider%20Interface%20Interface%20%28by%20ref%29))
  (if (not (Variables|Default|GetBeforeSpawn))
    (for _array_element SpawnedActor
      (Utilities|IsValid _array_element
        (:Is%20Valid
          (Ability|GameplayEffect|AddLooseGameplayTags _array_element (Variables|Default|GetGameplay%20Tags) (Variables|Default|GetShouldReplicate)))))
    (return)))
