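; BP_SetLeashToPointProvider is a spawn action that leashes newly spawned AI to the Living World point provider they came from.
; OnAllActorSpawned takes the provider's location and its max radius, raised to at least MinLeashSize.
; Athena AI bot controllers get a radial leash with an outer radius scaled by OuterLeashFactor.
; Other AI controllers fall back to setting center and radius on their FortAthenaLeashComponent, if present.
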(fn OnAllActorSpawned
  (params (SpawnedActor :type Array%20of%20Actor%20Object%20References) (SourcePointProvider :type Fort%20Athena%20Living%20World%20Point%20Provider%20Interface%20Interface%20%28by%20ref%29))
  (LocalVariables|SetSourcePointProviderLocation (Class|FortAthenaLivingWorldPointProviderInterface|GetPointProviderLocation SourcePointProvider))
  (LocalVariables|SetSourcePointProviderSize (Math|Float|Max(Float) (Class|FortAthenaLivingWorldPointProviderInterface|GetPointProviderMaxRadius SourcePointProvider) (LocalVariables|GetMinLeashSize)))
  (for _array_element SpawnedActor
    (bind _asfort_athena_aibot_controller (Utilities|Casting|CastToFortAthenaAIBotController (Fortnite|GetControllerfromActor _array_element))
      (:then
        (Class|BPFLAICommonFunctions|SetRadialLeash _asfort_athena_aibot_controller (LocalVariables|GetSourcePointProviderLocation) (LocalVariables|GetSourcePointProviderSize) (* (LocalVariables|GetSourcePointProviderSize) (LocalVariables|GetOuterLeashFactor))))
      (:CastFailed
        (bind _asaicontroller (Utilities|Casting|CastToAIController (Fortnite|GetControllerfromActor _array_element)))
        (Utilities|IsValid (Actor|GetComponentbyClass _asaicontroller "/Script/FortniteAI.FortAthenaLeashComponent")
          (:Is%20Valid
            (Class|FortAthenaLeashComponent|SetLeashCenter (Actor|GetComponentbyClass _asaicontroller "/Script/FortniteAI.FortAthenaLeashComponent") (LocalVariables|GetSourcePointProviderLocation))
            (Class|FortAthenaLeashComponent|SetLeashRadius (Actor|GetComponentbyClass _asaicontroller "/Script/FortniteAI.FortAthenaLeashComponent") (LocalVariables|GetSourcePointProviderSize)))))))
  (return))
